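# Post-process-style Control Rig for the large male DenseFog player body, correcting deformation of its wrist plate, front skirt and left shoulder pad.
# A sequence runs the shared DeformRig function and right-wrist correctives, then front-skirt fixes, then left shoulder-pad fixes.
# Spherical pose readers, pelvis-to-root height and deformRig anim attributes weight additive bone offsets, with interpolation smoothing some weights.
# Movement-state floats like crouch, slide, skydive, parachute and DBNO gate extra offsets.
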ControlRigBlueprint /BRCosmetics/Characters/Player/Male/Large/Bodies/M_LRG_DenseFog/Meshes/M_LRG_DenseFog_CtrlRig
graphs [
 RigVMFunctionLibrary {links [], nodes []}
 RigVMFunction_Sequence_3_1_ContainedGraph {nodes [
   n1 = RigVMFunctionEntryNode()
   n2 = &RigVMFunction_Sequence(ExecuteContext=@n1.ExecuteContext)
   n3 = &RigVMFunction_Sequence(ExecuteContext=@n2.B)
   RigVMFunctionReturnNode(A=@n2.A, B=@n3.A, C=@n3.B)
  ]}
 RigVMFunction_Sequence_4_ContainedGraph {nodes [
   n1 = RigVMFunctionEntryNode()
   RigVMFunction_Sequence = &RigVMFunction_Sequence(ExecuteContext=@n1.ExecuteContext)
   n2 = &RigVMFunction_Sequence(ExecuteContext=@RigVMFunction_Sequence.B)
   n3 = &RigVMFunction_Sequence(ExecuteContext=@n2.B)
   RigVMFunctionReturnNode(A=@RigVMFunction_Sequence.A, B=@n2.A, C=@n3.A, D=@n3.B)
  ]}
 RigVMModel {nodes [
   RigUnit_BeginExecution = &RigUnit_BeginExecution()
   RigVMFunction_Sequence_3 = RigVMAggregateNode(ExecuteContext=@RigUnit_BeginExecution.ExecutePin)
   FN_DeformRig_FNC = RigVMFunctionReferenceNode(
    DeformRig_ Curve="(deformRig_l_upperarm_up_alpha_72_#1=true,deformRig_r_upperarm_up_alpha_73_#2=False,deformRig_l_upperarm_fr_alpha_74_#3=true,deformRig_r_upperarm_fr_alpha_75_CB51B80C ...(truncated)"
    DeformRig_Bool="(upperarm_1_#4=True,elbow_3_#5=True,wrist_7_#6=True,groin_8_#7=True,glute_17_#8=True,knee_9_40C408324CA261 ...(truncated)"
    DeformRig_Debug=(upperarm_debug_18_#4=False,wrist_debug_20_#6=False,lat_debug_26_#9=False,groinGlute_debug_27_#7=False)
    DeformRig_Value="(upperarm_up_18_#10=(Rotation=(X=0,Y=0,Z=0,W=1),Translation=(X=0,Y=0,Z=8),Scale3D=(X=0.7,Y=1,Z=1)),upperarm_fwd_fr_19_#11=(Rotation=(X=0,Y=0,Z=0,W=1),Translation=(X=0,Y=10,Z=0),Scale3D=(X=0.5,Y=1,Z= ...(truncated)"
    ExecuteContext=@RigVMFunction_Sequence_3.A
   )
   > &RigUnit_ParentConstraint(
    AdvancedSettings=(InterpolationType=Average,RotationOrderForFilter=XZY)
    Child=(Type=Bone,Name="wrist_plate_support_base")
    Parents=((Item=(Type=Bone,Name="None"),Weight=1),(Item=(Type=Bone,Name="hand_r"),Weight=0.8),(Item=(Type=Bone,Name="lowerarm_twist_01_r"),Weight=0.2))
    Weight=1
   )
   > n1 = &RigUnit_SphericalPoseReader(
    ActiveRegionScaleFactors=(PositiveWidth=1,NegativeWidth=1,PositiveHeight=1,NegativeHeight=1)
    ActiveRegionSize=0.1
    Debug=(bDrawDebug=false,DebugScale=25,DebugSegments=20,DebugThickness=0.25)
    DriverAxis=(X=-1,Y=0,Z=0)
    DriverItem=(Type=Bone,Name="hand_r")
    FalloffRegionScaleFactors=(PositiveWidth=2,NegativeWidth=2,PositiveHeight=2,NegativeHeight=2)
    FalloffSize=0.3
    FlipHeightScaling=true
    FlipWidthScaling=false
    OptionalParentItem=(Type=Bone,Name="lowerarm_twist_02_r")
    RotationOffset=(X=0,Y=100,Z=60)
   )
   n2 = &RigVMFunction_AlphaInterp(
    InterpSpeedDecreasing=12
    InterpSpeedIncreasing=12
    OutRange=(Max=1)
    Scale=1
    bClampResult=true
    bMapRange=true
    Bias=~
    ClampMin=~
    Value=@n1.OutputParam
   )
   > &RigUnit_ModifyTransforms(
    ItemToModify=((Item=(Type=Bone,Name="wrist_plate_support_1"),Transform=(Rotation=(X=0,Y=0,Z=-0.1288,W=0.9917),Translation=(X=-1.6458,Y=-1.7747,Z=0.232),Scale3D=(X=1,Y=1,Z=1))))
    WeightMinimum=0
    Weight=@n2.Result
   )
   n3 = &RigVMFunction_AlphaInterp(
    InterpSpeedDecreasing=12
    InterpSpeedIncreasing=12
    OutRange=(Max=1)
    Scale=1
    bClampResult=true
    bMapRange=true
    Bias=~
    ClampMin=~
    Value=&RigVMFunction_MathDoubleMul(A=@n2.Result, B=RigVMVariableNode() {variable=Is Skydive Float}.Value).Result
   )
   > &RigUnit_ModifyTransforms(
    ItemToModify=((Item=(Type=Bone,Name="wrist_plate_support_1"),Transform=(Rotation=(X=0.0631,Y=0.003,Z=0.0478,W=0.9969),Translation=(X=-1.6458,Y=-1.7747,Z=0.232),Scale3D=(X=1,Y=1,Z=1))))
    WeightMinimum=0
    Weight=@n3.Result
   )
   > n4 = &RigUnit_SphericalPoseReader(
    ActiveRegionScaleFactors=(PositiveWidth=1.5,NegativeWidth=1,PositiveHeight=1,NegativeHeight=1)
    ActiveRegionSize=0.1
    Debug=(bDrawDebug=false,DebugScale=25,DebugSegments=20,DebugThickness=0.25)
    DriverAxis=(X=0,Y=-1,Z=0)
    DriverItem=(Type=Bone,Name="hand_r")
    FalloffRegionScaleFactors=(PositiveWidth=2,NegativeWidth=0.5,PositiveHeight=1,NegativeHeight=0.5)
    FalloffSize=0.3
    FlipHeightScaling=true
    FlipWidthScaling=false
    OptionalParentItem=(Type=Bone,Name="lowerarm_twist_02_r")
    RotationOffset=(X=0,Y=-60,Z=0)
   )
   n5 = &RigVMFunction_AlphaInterp(
    InterpSpeedDecreasing=12
    InterpSpeedIncreasing=12
    OutRange=(Max=1)
    Scale=1
    bClampResult=true
    bMapRange=true
    Bias=~
    ClampMin=~
    Value=@n4.OutputParam
   )
   > &RigUnit_ModifyTransforms(
    ItemToModify=((Item=(Type=Bone,Name="wrist_plate_support_1"),Transform=(Rotation=(X=-0.0555,Y=-0.0233,Z=0.0928,W=0.9939),Translation=(X=-4,Y=-2,Z=0.8474),Scale3D=(X=1,Y=1,Z=1))))
    WeightMinimum=0
    Weight=@n5.Result
   )
   &RigUnit_ParentConstraint(
    AdvancedSettings=(InterpolationType=Average,RotationOrderForFilter=XZY)
    Child=(Type=Bone,Name="dyn_skirt_frnt_base")
    Parents=((Item=(Type=Bone,Name="thigh_l"),Weight=0.5),(Item=(Type=Bone,Name="thigh_r"),Weight=0.5))
    Weight=0.9
    ExecutePin=@RigVMFunction_Sequence_3.B
   )
   > n6 = &RigUnit_SphericalPoseReader(
    ActiveRegionScaleFactors=(PositiveWidth=2,NegativeWidth=1,PositiveHeight=2,NegativeHeight=2)
    ActiveRegionSize=0.1
    Debug=(bDrawDebug=true,bDraw2D=False,bDrawLocalAxes=False,DebugScale=25,DebugSegments=20,DebugThickness=0.25)
    DriverAxis=(X=-1,Y=0,Z=0)
    DriverItem=(Type=Bone,Name="thigh_l")
    FalloffRegionScaleFactors=(PositiveWidth=1,NegativeWidth=1,PositiveHeight=1,NegativeHeight=1)
    FalloffSize=0.4
    FlipHeightScaling=False
    FlipWidthScaling=False
    OptionalParentItem=(Type=Bone,Name="spine_01")
    RotationOffset=(X=112,Y=0,Z=0)
   )
   > &RigUnit_ModifyTransforms(
    ItemToModify=((Item=(Type=Bone,Name="dyn_skirt_frnt_base"),Transform=(Rotation=(X=-0.0744,Y=0,Z=0,W=0.9972),Translation=(X=0,Y=0,Z=-5),Scale3D=(X=1,Y=1,Z=1))))
    WeightMinimum=0
    Weight=@n6.OutputParam
   )
   > n7 = &RigUnit_SphericalPoseReader(
    ActiveRegionScaleFactors=(PositiveWidth=1,NegativeWidth=2,PositiveHeight=2,NegativeHeight=2)
    ActiveRegionSize=0.1
    Debug=(bDrawDebug=true,bDraw2D=False,bDrawLocalAxes=False,DebugScale=25,DebugSegments=20,DebugThickness=0.25)
    DriverAxis=(X=1,Y=0,Z=0)
    DriverItem=(Type=Bone,Name="thigh_r")
    FalloffRegionScaleFactors=(PositiveWidth=0.5,NegativeWidth=1,PositiveHeight=1,NegativeHeight=1)
    FalloffSize=0.5
    FlipHeightScaling=False
    FlipWidthScaling=False
    OptionalParentItem=(Type=Bone,Name="spine_01")
    RotationOffset=(X=-60,Y=0,Z=10)
   )
   > RigVMVariableNode(Value=&RigVMFunction_MathDoubleRemap(SourceMaximum=1, SourceMinimum=0, TargetMaximum=1, TargetMinimum=0, Value=RigVMVariableNode() {variable=Is Right Left Movement}.Value).Result) {variable=Right Movement}
   n8 = RigVMVariableNode() {variable=Right Movement}
   n9 = RigVMVariableNode() {variable=Is Crouching}
   n10 = &RigVMFunction_AlphaInterp(
    InterpSpeedDecreasing=12
    InterpSpeedIncreasing=12
    OutRange=(Max=1)
    Scale=1
    bClampResult=true
    bMapRange=true
    Bias=~
    ClampMin=~
    Value=&RigVMFunction_MathDoubleMul(B=&RigVMFunction_MathDoubleMul(A=@n8.Value, B=@n9.Value).Result, A=@n7.OutputParam).Result
   )
   > &RigUnit_ModifyTransforms(
    ItemToModify=((Item=(Type=Bone,Name="dyn_skirt_frnt_base"),Transform=(Rotation=(X=0.1359,Y=-0.0199,Z=0.1776,W=0.9745),Translation=(X=2,Y=2,Z=4),Scale3D=(X=1,Y=1,Z=1))))
    WeightMinimum=0
    Weight=@n10.Result
   )
   Remap = &RigVMFunction_MathDoubleRemap(SourceMaximum=1, SourceMinimum=0, TargetMaximum=1, TargetMinimum=0, Value=RigVMVariableNode() {variable=Is Backward Forward MOvement}.Value)
   > RigVMVariableNode(Value=@Remap.Result) {variable=Forward Movement}
   n11 = RigVMVariableNode() {variable=Is Crouching}
   n12 = RigVMVariableNode() {variable=Forward Movement}
   > &RigUnit_ModifyTransforms(
    ItemToModify=((Item=(Type=Bone,Name="dyn_skirt_frnt_base"),Transform=(Rotation=(X=0.1359,Y=-0.0199,Z=0.1776,W=0.9745),Translation=(X=2,Y=2,Z=4),Scale3D=(X=1,Y=1,Z=1))))
    WeightMinimum=0
    Weight=&RigVMFunction_MathDoubleMul(B=&RigVMFunction_MathDoubleMul(B=@n11.Value, A=@n12.Value).Result, A=@n7.OutputParam).Result
   )
   n13 = &RigUnit_GetTransform(Item=(Type=Bone,Name="root"), bInitial=False)
   n14 = &RigUnit_GetTransform(Item=(Type=Bone,Name="pelvis"), bInitial=False)
   n15 = &RigVMFunction_AlphaInterp(
    InterpSpeedDecreasing=8
    InterpSpeedIncreasing=8
    OutRange=(Max=1)
    Scale=1
    bClampResult=true
    bMapRange=true
    Bias=~
    ClampMin=~
    Value=&RigVMFunction_MathDoubleRemap(
     SourceMaximum=22
     SourceMinimum=32
     TargetMaximum=1
     TargetMinimum=0
     Value=&RigVMFunction_MathVectorDistance(A=(X=0,Y=0,Z=0), B=(X=0,Y=0,Z=0), B.Z=@n13.Transform.Translation.Z, A.Z=@n14.Transform.Translation.Z).Result
    ).Result
   )
   > &RigUnit_ModifyTransforms(
    ItemToModify=((Item=(Type=Bone,Name="dyn_skirt_frnt_base"),Transform=(Rotation=(X=0.0151,Y=-0.0858,Z=0.173,W=0.9811),Translation=(X=-2,Y=0,Z=0),Scale3D=(X=1,Y=1,Z=1))))
    WeightMinimum=0
    Weight=@n15.Result
   )
   n16 = &RigVMFunction_MathVectorDistance(
    A=(X=0,Y=0,Z=0)
    B=(X=0,Y=0,Z=0)
    B.Z=&RigUnit_GetTransform(Item=(Type=Bone,Name="root"), bInitial=False).Transform.Translation.Z
    A.Z=&RigUnit_GetTransform(Item=(Type=Bone,Name="pelvis"), bInitial=False).Transform.Translation.Z
   )
   n17 = &RigVMFunction_AlphaInterp(
    InterpSpeedDecreasing=8
    InterpSpeedIncreasing=8
    OutRange=(Max=1)
    Scale=1
    bClampResult=true
    bMapRange=true
    Bias=~
    ClampMin=~
    Value=&RigVMFunction_MathDoubleMul(
     A=&RigVMFunction_MathDoubleRemap(SourceMaximum=22, SourceMinimum=32, TargetMaximum=1, TargetMinimum=0, Value=@n16.Result).Result
     B=RigVMVariableNode() {variable=Is Baseball Sliding}.Value
    ).Result
   )
   > &RigUnit_ModifyTransforms(
    ItemToModify=((Item=(Type=Bone,Name="dyn_skirt_frnt_base"),Transform=(Rotation=(X=0.0842,Y=-0.0226,Z=-0.2578,W=0.9623),Translation=(X=2,Y=0,Z=6),Scale3D=(X=1,Y=1,Z=1))))
    WeightMinimum=0
    Weight=@n17.Result
   )
   &RigUnit_ModifyTransforms(
    ItemToModify=((Item=(Type=Bone,Name="deform_shoulderpad_l_base"),Transform=(Rotation=(X=0,Y=0,Z=0,W=1),Translation=(X=0,Y=0,Z=-1),Scale3D=(X=1,Y=1.1,Z=1))))
    Weight=1
    WeightMinimum=0
    ExecutePin=@RigVMFunction_Sequence_3.C
   )
   > RigVMFunction_Sequence = RigVMAggregateNode()
   n18 = RigVMDispatchNode(BoneName=root, Default=0, Name=deformRig_l_upperarm_up_alpha) {template="DISPATCH_RigDispatch_GetAnimAttribute(in Name,in BoneName,in Default,out Value,out Found)"}
   &RigUnit_ModifyTransforms(
    ItemToModify=((Item=(Type=Bone,Name="deform_shoulderpad_l"),Transform=(Rotation=(X=0.1784,Y=-0.2886,Z=0.0931,W=0.9361),Translation=(X=7,Y=-5.8911,Z=0),Scale3D=(X=0.9,Y=0.9,Z=1))))
    WeightMinimum=0
    Weight=@n18.Value
    ExecutePin=@RigVMFunction_Sequence.A
   )
   > &RigUnit_ModifyTransforms(
    ItemToModify=((Item=(Type=Bone,Name="deform_shoulderpad_l"),Transform=(Rotation=(X=-0.113,Y=0.1077,Z=-0.1603,W=0.9746),Translation=(X=0,Y=4,Z=2),Scale3D=(X=0.9,Y=0.9,Z=1))))
    WeightMinimum=0
    Weight=&RigVMFunction_MathDoubleMul(B=@n18.Value, A=RigVMVariableNode() {variable=Is Parachute Open}.Value).Result
   )
   > &RigUnit_ModifyTransforms(
    ItemToModify=((Item=(Type=Bone,Name="deform_shoulderpad_l"),Transform=(Rotation=(X=-0.083,Y=0.0905,Z=-0.0357,W=0.9918),Translation=(X=0,Y=3,Z=0),Scale3D=(X=0.9,Y=0.9,Z=1))))
    WeightMinimum=0
    Weight=&RigVMFunction_MathDoubleMul(A=@n18.Value, B=RigVMVariableNode() {variable=Is Skydive Float}.Value).Result
   )
   n19 = RigVMDispatchNode(BoneName=root, Default=0, Name=deformRig_l_upperarm_fr_alpha) {template="DISPATCH_RigDispatch_GetAnimAttribute(in Name,in BoneName,in Default,out Value,out Found)"}
   &RigUnit_ModifyTransforms(
    ItemToModify=((Item=(Type=Bone,Name="deform_shoulderpad_l"),Transform=(Rotation=(X=0.2586,Y=0.0113,Z=0.0421,W=0.965),Translation=(X=0,Y=-2,Z=2),Scale3D=(X=0.9,Y=1,Z=1))))
    WeightMinimum=0
    Weight=@n19.Value
    ExecutePin=@RigVMFunction_Sequence.B
   )
   Multiply = &RigVMFunction_MathDoubleMul(A=@n19.Value, B=RigVMVariableNode() {variable=Is DBNO}.Value)
   > &RigUnit_ModifyTransforms(
    ItemToModify=((Item=(Type=Bone,Name="deform_shoulderpad_l"),Transform=(Rotation=(X=0,Y=0,Z=0.0436,W=0.999),Translation=(X=0,Y=-2,Z=0),Scale3D=(X=1,Y=1,Z=1))))
    WeightMinimum=0
    Weight=@Multiply.Result
   )
   n20 = &RigUnit_SphericalPoseReader(
    ActiveRegionScaleFactors=(PositiveWidth=1,NegativeWidth=1,PositiveHeight=1,NegativeHeight=1)
    ActiveRegionSize=0.05
    Debug=(bDrawDebug=true,bDraw2D=False,bDrawLocalAxes=False,DebugScale=25,DebugSegments=20,DebugThickness=0.25)
    DriverAxis=(X=1,Y=0,Z=0)
    DriverItem=(Type=Bone,Name="clavicle_l")
    FalloffRegionScaleFactors=(PositiveWidth=1,NegativeWidth=1,PositiveHeight=1,NegativeHeight=1)
    FalloffSize=0.3
    FlipHeightScaling=False
    FlipWidthScaling=False
    OptionalParentItem=(Type=Bone,Name="None")
    RotationOffset=(X=0,Y=-30,Z=0)
    ExecutePin=@RigVMFunction_Sequence.C
   )
   > RigVMVariableNode(Value=@n20.OutputParam) {variable=Clavicle_L_raise}
   > &RigUnit_ModifyTransforms(
    ItemToModify=((Item=(Type=Bone,Name="deform_shoulderpad_l"),Transform=(Rotation=(X=0.0509,Y=-0.083,Z=0.0905,W=0.9911),Translation=(X=4,Y=0,Z=-3),Scale3D=(X=1,Y=1,Z=1))))
    WeightMinimum=0
    Weight=@n20.OutputParam
   )
   n21 = &RigUnit_SphericalPoseReader(
    ActiveRegionScaleFactors=(PositiveWidth=1,NegativeWidth=1,PositiveHeight=1,NegativeHeight=1)
    ActiveRegionSize=0.05
    Debug=(bDrawDebug=false,bDraw2D=False,bDrawLocalAxes=False,DebugScale=25,DebugSegments=20,DebugThickness=0.25)
    DriverAxis=(X=1,Y=0,Z=0)
    DriverItem=(Type=Bone,Name="upperarm_l")
    FalloffRegionScaleFactors=(PositiveWidth=1,NegativeWidth=1,PositiveHeight=1,NegativeHeight=1)
    FalloffSize=0.25
    FlipHeightScaling=False
    FlipWidthScaling=False
    OptionalParentItem=(Type=Bone,Name="None")
    RotationOffset=(X=0,Y=-110,Z=80)
    ExecutePin=@RigVMFunction_Sequence.D
   )
   > n22 = &RigUnit_SphericalPoseReader(
    ActiveRegionScaleFactors=(PositiveWidth=1,NegativeWidth=1,PositiveHeight=1,NegativeHeight=1)
    ActiveRegionSize=0.05
    Debug=(bDrawDebug=false,bDraw2D=False,bDrawLocalAxes=False,DebugScale=25,DebugSegments=20,DebugThickness=0.25)
    DriverAxis=(X=1,Y=0,Z=0)
    DriverItem=(Type=Bone,Name="upperarm_l")
    FalloffRegionScaleFactors=(PositiveWidth=1,NegativeWidth=1,PositiveHeight=1,NegativeHeight=1)
    FalloffSize=0.2
    FlipHeightScaling=False
    FlipWidthScaling=False
    OptionalParentItem=(Type=Bone,Name="spine_05")
    RotationOffset=(X=0,Y=-110,Z=130)
   )
   Add = &RigVMFunction_MathDoubleAdd(A=@n21.OutputParam, B=@n22.OutputParam)
   > &RigUnit_ModifyTransforms(
    ItemToModify="((Item=(Type=Bone,Name=\"deform_shoulderpad_l\"),Transform=(Rotation=(X=0,Y=-0.0872,Z=0,W=0.9962),Translation=(X=0,Y=-0.6648,Z=2),Scale3D=(X=1,Y=1,Z=1))),(Item=(Type=Bone,Name=\"shoulderpad_02_l_base\"),Transform=(Rotation=(X=0,Y=0,Z=0,W=1),Translation=(X=0,Y= ...(truncated)"
    WeightMinimum=0
    Weight=&RigVMFunction_MathDoubleMul(A=@Add.Result, B=RigVMVariableNode() {variable=Is Crouching}.Value).Result
   )
  ]}
]
hierarchy [
 SPR_Thigh_R {type=Curve, loc=<0 0 0>, rot=<0 0 0 1>}
 Root {loc=<0 0 0>, rot=<0 0 0 1>}
 pelvis {parent=Root, loc=<0 -0.7321 92.4218>, rot=<0.0127 0.707 -0.0127 -0.707>}
 spine_01 {parent=pelvis, loc=<5.6002 -0.5128 0>, rot=<0 0 0.034 -0.9994>}
 spine_02 {parent=spine_01, loc=<8.0753 0.0139 0>, rot=<0 0 0.0238 -0.9997>}
 spine_03 {parent=spine_02, loc=<7.0393 0 0>, rot=<0 0 0.0149 -0.9999>}
 spine_04 {parent=spine_03, loc=<8.1024 0 0>, rot=<0 0 -0.0132 -0.9999>}
 spine_05 {parent=spine_04, loc=<10.0146 0 0>, rot=<0 0 -0.0015 -1>}
 attach_cape {parent=spine_05, loc=<9.1509 -13.3783 0>, rot=<0.0616 -0.7044 -0.0616 -0.7044>}
 attach_backpack {parent=spine_05, loc=<-0.123 -15.1423 0>, rot=<-0.0093 -0.707 0.0093 -0.707>}
 neck_01 {parent=spine_05, loc=<10.815 1.9196 0>, rot=<0 0 -0.1864 -0.9825>}
 neck_02 {parent=neck_01, loc=<4.7347 0.3045 0>, rot=<0 0 0.0363 -0.9993>}
 head {parent=neck_02, loc=<5.7632 0.5552 0>, rot=<0 0 0.115 -0.9934>}
 dyn_helmet_back_base {parent=head, loc=<9.9656 -8.2161 0>, rot=<0.5021 -0.4979 0.4979 -0.5021>}
 dyn_helmet_back_1 {parent=dyn_helmet_back_base, loc=<0 0 0>, rot=<0 0 0 1>}
 dyn_coif_back_base {parent=head, loc=<4.8201 -6.6846 0>, rot=<0.269 -0.9632 0 0>}
 dyn_coif_back_1 {parent=dyn_coif_back_base, loc=<0 0 0>, rot=<0 0 0 1>}
 dyn_coif_back_2 {parent=dyn_coif_back_1, loc=<3.8769 0 0>, rot=<0 0 0.04 -0.9992>}
 dyn_coif_side_l_base {parent=head, loc=<4.9303 -4.9121 -5.3125>, rot=<0.2996 -0.8921 -0.326 -0.0896>}
 dyn_coif_side_l_1 {parent=dyn_coif_side_l_base, loc=<0 0 0>, rot=<0 0 0 1>}
 dyn_coif_side_l_2 {parent=dyn_coif_side_l_1, loc=<3.8769 0 0>, rot=<0 0 0.04 -0.9992>}
 dyn_coif_side_r_base {parent=head, loc=<4.9305 -4.9121 5.3124>, rot=<0.8921 0.2996 -0.0896 0.326>}
 dyn_coif_side_r_1 {parent=dyn_coif_side_r_base, loc=<0 0 0>, rot=<0 0 0 1>}
 dyn_coif_side_r_2 {parent=dyn_coif_side_r_1, loc=<-3.8769 0 0>, rot=<0 0 0.04 -0.9992>}
 dyn_helmet_frnt_base {parent=head, loc=<16.2014 7.8561 0>, rot=<0.6214 -0.3374 0.3374 -0.6214>}
 dyn_helmet_frnt_1 {parent=dyn_helmet_frnt_base, loc=<0 0 0>, rot=<0 0 0 1>}
 dyn_helmet_mid_1_base {parent=head, loc=<15.0136 -4.4723 0>, rot=<0.6788 -0.198 0.198 -0.6788>}
 dyn_helmet_mid_1_1 {parent=dyn_helmet_mid_1_base, loc=<0 0 0>, rot=<0 0 0 1>}
 dyn_helmet_mid_2_base {parent=head, loc=<17.5569 2.5244 0>, rot=<0.7031 -0.0754 0.0754 -0.7031>}
 dyn_helmet_mid_2_1 {parent=dyn_helmet_mid_2_base, loc=<0 0 0>, rot=<0 0 0 1>}
 clavicle_l {parent=spine_05, loc=<2.5856 -0.2866 -4.3777>, rot=<0.0465 -0.7289 -0.009 -0.6829>}
 upperarm_l {parent=clavicle_l, loc=<18.2264 0.1342 -0.001>, rot=<0.0253 -0.2693 0.0013 -0.9627>}
 lowerarm_l {parent=upperarm_l, loc=<29.0071 0.1686 -1.239>, rot=<0.0391 -0.0099 -0.2453 -0.9686>}
 hand_l {parent=lowerarm_l, loc=<25.8607 0.222 1.0274>, rot=<-0.7193 -0.0988 0.1193 -0.6773>}
 weapon_l {parent=hand_l, loc=<-0.1974 -3.4111 0.3811>, rot=<0 0 0 1>}
 thumb_01_l {parent=hand_l, loc=<1.3118 -0.3191 -2.459>, rot=<0.6546 -0.4042 0.0664 -0.6354>}
 thumb_02_l {parent=thumb_01_l, loc=<5.9336 -0.0448 0.0401>, rot=<-0.0037 0.0213 0.298 -0.9543>}
 thumb_03_l {parent=thumb_02_l, loc=<2.2223 0.3682 -0.0907>, rot=<0.0118 -0.013 0.0066 -0.9998>}
 index_metacarpal_l {parent=hand_l, loc=<1.6535 0.0183 -2.96>, rot=<0.0158 -0.0124 -0.0173 -0.9996>}
 index_01_l {parent=index_metacarpal_l, loc=<8.3008 0.0357 -0.0524>, rot=<-0.0044 0.0485 0.0924 -0.9945>}
 index_02_l {parent=index_01_l, loc=<3.3587 -0.3004 0.0452>, rot=<-0.002 0.0027 0.3225 -0.9466>}
 index_03_l {parent=index_02_l, loc=<2.5578 -0.123 -0.0252>, rot=<-0.002 0.0067 0.2558 -0.9667>}
 middle_metacarpal_l {parent=hand_l, loc=<2.1856 -0.1725 -0.5984>, rot=<0.006 0.0168 -0.0329 -0.9993>}
 middle_01_l {parent=middle_metacarpal_l, loc=<7.3275 -0.2447 0.0129>, rot=<-0.0208 0.0042 0.0858 -0.9961>}
 middle_02_l {parent=middle_01_l, loc=<4.0123 0.1401 0.1175>, rot=<-0.0218 0.0258 0.3512 -0.9357>}
 middle_03_l {parent=middle_02_l, loc=<2.6077 0.0534 0.1373>, rot=<0.0082 0.0188 0.1708 -0.9851>}
 ring_metacarpal_l {parent=hand_l, loc=<2.2074 -0.3625 1.3638>, rot=<-0.0284 0.0722 -0.0458 -0.9959>}
 ring_01_l {parent=ring_metacarpal_l, loc=<7.0295 -0.6197 -0.1446>, rot=<-0.0489 -0.0252 0.1892 -0.9804>}
 ring_02_l {parent=ring_01_l, loc=<3.6174 0.4647 -0.1027>, rot=<0 0 0.2806 -0.9598>}
 ring_03_l {parent=ring_02_l, loc=<2.4152 0.114 0.0199>, rot=<0 0 0.1189 -0.9929>}
 pinky_metacarpal_l {parent=hand_l, loc=<1.7616 -0.9753 3.3814>, rot=<-0.0397 0.0814 -0.069 -0.9935>}
 pinky_01_l {parent=pinky_metacarpal_l, loc=<7.1709 -0.4831 -0.1094>, rot=<-0.0654 -0.0035 0.2242 -0.9723>}
 pinky_02_l {parent=pinky_01_l, loc=<2.8274 0.0893 -0.1043>, rot=<0.0037 0.0011 0.2854 -0.9584>}
 pinky_03_l {parent=pinky_02_l, loc=<2.2046 0.0307 -0.0078>, rot=<-0.0002 0.0098 0.1445 -0.9895>}
 hand_attach_l {parent=hand_l, loc=<8.5 -3 0>, rot=<0.7071 0.7071 0 0>}
 deform_wrist_dn_base_l {parent=hand_l, loc=<0 -3.4173 0>, rot=<0 0 0 -1>}
 deform_wrist_dn_l {parent=deform_wrist_dn_base_l, loc=<0 0 0>, rot=<0 0 0 1>}
 deform_wrist_up_base_l {parent=hand_l, loc=<0 2.7805 0>, rot=<0 0 0 -1>}
 deform_wrist_up_l {parent=deform_wrist_up_base_l, loc=<0 0 0>, rot=<0 0 0 1>}
 lowerarm_twist_02_l {parent=lowerarm_l, loc=<9.6199 0 0>, rot=<0 0 0 -1>}
 lowerarm_twist_01_l {parent=lowerarm_l, loc=<17.8549 0 0>, rot=<0 0 0 -1>}
 deform_elbow_in_base_l {parent=lowerarm_l, loc=<0.2112 5.645 0.8274>, rot=<0 0 0 -1>}
 deform_elbow_in_l {parent=deform_elbow_in_base_l, loc=<0 0 0>, rot=<0 0 0 1>}
 deform_elbow_out_base_l {parent=lowerarm_l, loc=<0 -7.5129 0>, rot=<0 0 0 -1>}
 deform_elbow_out_l {parent=deform_elbow_out_base_l, loc=<0 0 0>, rot=<0 0 0 1>}
 upperarm_twist_01_l {parent=upperarm_l, loc=<7.9796 0.0843 -0.6195>, rot=<0 0 0 -1>}
 deform_upperarm_bk_base_l {parent=upperarm_twist_01_l, loc=<-6.8983 -13.3826 -1.3258>, rot=<0 0 0 1>}
 deform_upperarm_bk_l {parent=deform_upperarm_bk_base_l, loc=<0 0 0>, rot=<0 0 0 1>}
 deform_upperarm_fr_base_l {parent=upperarm_twist_01_l, loc=<-8.3487 10.7117 -2.1455>, rot=<0 0 0 1>}
 deform_upperarm_fr_l {parent=deform_upperarm_fr_base_l, loc=<0 0 0>, rot=<0 0 0 1>}
 deform_upperarm_up_base_l {parent=upperarm_twist_01_l, loc=<-11.2012 -1.2234 8.7784>, rot=<0 0 0 1>}
 deform_upperarm_up_l {parent=deform_upperarm_up_base_l, loc=<0 0 0>, rot=<0 0 0 1>}
 upperarm_twist_02_l {parent=upperarm_l, loc=<19.5059 0.0843 -0.6195>, rot=<0 0 0 -1>}
 deform_bicep_base_l {parent=upperarm_twist_02_l, loc=<0 8.0494 0>, rot=<0 0 0 1>}
 deform_bicep_l {parent=deform_bicep_base_l, loc=<0 0 0>, rot=<0 0 0 1>}
 dyn_shoulderpad_l_rotate {parent=upperarm_l, loc=<0 0 0>, rot=<0 0 0 1>}
 dyn_shoulderpad_l_base {parent=dyn_shoulderpad_l_rotate, loc=<-8.0758 -0.0145 11.349>, rot=<-0.0111 0.057 -0.0281 -0.9979>}
 dyn_shoulderpad_l_1 {parent=dyn_shoulderpad_l_base, loc=<0 0 0>, rot=<0 0 0 1>}
 plate_l_rotate {parent=upperarm_l, loc=<0 0 0>, rot=<0 0 0 1>}
 plate_01_l_base {parent=plate_l_rotate, loc=<-0.3126 0.3789 14.0918>, rot=<-0.0103 0.0285 -0.0284 -0.9991>}
 plate_01_l_1 {parent=plate_01_l_base, loc=<0 0 0>, rot=<0 0 0 1>}
 plate_02_l_base {parent=plate_l_rotate, loc=<-11.2399 -0.2051 11.8219>, rot=<-0.0123 0.1018 -0.0275 -0.9943>}
 plate_02_l_1 {parent=plate_02_l_base, loc=<0 0 0>, rot=<0 0 0 1>}
 plate_03_l_base {parent=plate_l_rotate, loc=<-7.3936 17.3023 0.977>, rot=<0.4081 0.0434 -0.0096 -0.9119>}
 plate_03_l_1 {parent=plate_03_l_base, loc=<0 0 0>, rot=<0 0 0 1>}
 deform_shoulderpad_l_base {parent=clavicle_l, loc=<10.1663 4.5948 4.6433>, rot=<0.0008 0.0336 -0.0252 -0.9991>}
 deform_shoulderpad_l {parent=deform_shoulderpad_l_base, loc=<0 0 0>, rot=<0 0 0 1>}
 shoulderpad_01_l_base {parent=deform_shoulderpad_l, loc=<15.3767 -4.1811 6.499>, rot=<0 -0.2749 0 -0.9615>}
 shoulderpad_01_l {parent=shoulderpad_01_l_base, loc=<0 0 0>, rot=<0 0 0 1>}
 shoulderpad_02_l_base {parent=deform_shoulderpad_l, loc=<2.6443 12.8685 -0.9667>, rot=<0.4031 -0.2442 -0.1108 -0.875>}
 shoulderpad_02_l {parent=shoulderpad_02_l_base, loc=<0 0 0>, rot=<0 0 0 1>}
 clavicle_r {parent=spine_05, loc=<2.5855 -0.2866 4.3777>, rot=<0.7289 0.0465 -0.6829 0.009>}
 upperarm_r {parent=clavicle_r, loc=<-18.2264 -0.1342 0.001>, rot=<0.0253 -0.2693 0.0013 -0.9627>}
 lowerarm_r {parent=upperarm_r, loc=<-29.0071 -0.1686 1.239>, rot=<0.0391 -0.0099 -0.2453 -0.9686>}
 hand_r {parent=lowerarm_r, loc=<-25.8607 -0.222 -1.0274>, rot=<-0.7193 -0.0988 0.1193 -0.6773>}
 weapon_r {parent=hand_r, loc=<0.1974 3.4111 -0.3811>, rot=<0 0 0 1>}
 thumb_01_r {parent=hand_r, loc=<-1.3118 0.3191 2.459>, rot=<0.6546 -0.4042 0.0664 -0.6354>}
 thumb_02_r {parent=thumb_01_r, loc=<-5.9336 0.0448 -0.0401>, rot=<-0.0037 0.0213 0.298 -0.9543>}
 thumb_03_r {parent=thumb_02_r, loc=<-2.2223 -0.3682 0.0907>, rot=<0.0118 -0.013 0.0066 -0.9998>}
 index_metacarpal_r {parent=hand_r, loc=<-1.6535 -0.0183 2.96>, rot=<0.0158 -0.0124 -0.0173 -0.9996>}
 index_01_r {parent=index_metacarpal_r, loc=<-8.3008 -0.0357 0.0524>, rot=<-0.0044 0.0485 0.0924 -0.9945>}
 index_02_r {parent=index_01_r, loc=<-3.3587 0.3004 -0.0452>, rot=<-0.002 0.0027 0.3225 -0.9466>}
 index_03_r {parent=index_02_r, loc=<-2.5578 0.123 0.0252>, rot=<-0.002 0.0067 0.2558 -0.9667>}
 middle_metacarpal_r {parent=hand_r, loc=<-2.1856 0.1725 0.5984>, rot=<0.006 0.0168 -0.0329 -0.9993>}
 middle_01_r {parent=middle_metacarpal_r, loc=<-7.3275 0.2447 -0.0129>, rot=<-0.0208 0.0042 0.0858 -0.9961>}
 middle_02_r {parent=middle_01_r, loc=<-4.0123 -0.1401 -0.1175>, rot=<-0.0218 0.0258 0.3512 -0.9357>}
 middle_03_r {parent=middle_02_r, loc=<-2.6077 -0.0534 -0.1373>, rot=<0.0082 0.0188 0.1708 -0.9851>}
 ring_metacarpal_r {parent=hand_r, loc=<-2.2074 0.3625 -1.3638>, rot=<-0.0284 0.0722 -0.0458 -0.9959>}
 ring_01_r {parent=ring_metacarpal_r, loc=<-7.0295 0.6197 0.1446>, rot=<-0.0489 -0.0252 0.1892 -0.9804>}
 ring_02_r {parent=ring_01_r, loc=<-3.6174 -0.4647 0.1027>, rot=<0 0 0.2806 -0.9598>}
 ring_03_r {parent=ring_02_r, loc=<-2.4152 -0.114 -0.0199>, rot=<0 0 0.1189 -0.9929>}
 pinky_metacarpal_r {parent=hand_r, loc=<-1.7616 0.9753 -3.3814>, rot=<-0.0397 0.0814 -0.069 -0.9935>}
 pinky_01_r {parent=pinky_metacarpal_r, loc=<-7.1709 0.4831 0.1094>, rot=<-0.0654 -0.0035 0.2242 -0.9723>}
 pinky_02_r {parent=pinky_01_r, loc=<-2.8274 -0.0893 0.1043>, rot=<0.0037 0.0011 0.2854 -0.9584>}
 pinky_03_r {parent=pinky_02_r, loc=<-2.2046 -0.0307 0.0078>, rot=<-0.0002 0.0098 0.1445 -0.9895>}
 hand_attach_r {parent=hand_r, loc=<-8.5 3 0>, rot=<0 0 -0.7071 -0.7071>}
 deform_wrist_dn_base_r {parent=hand_r, loc=<0 3.4172 0>, rot=<0 0 0 -1>}
 deform_wrist_dn_r {parent=deform_wrist_dn_base_r, loc=<0 0 0>, rot=<0 0 0 -1>}
 deform_wrist_up_base_r {parent=hand_r, loc=<-0.0001 -2.7804 0>, rot=<0 0 0 -1>}
 deform_wrist_up_r {parent=deform_wrist_up_base_r, loc=<0 0 0>, rot=<0 0 0 -1>}
 wrist_plate_support_rotate {parent=hand_r, loc=<0 0 0>, rot=<0 0 0 1>}
 wrist_plate_support_base {parent=wrist_plate_support_rotate, loc=<0 -4.7185 0>, rot=<0 0 0 1>}
 wrist_plate_support_1 {parent=wrist_plate_support_base, loc=<0 0 0>, rot=<0 0 0 1>}
 lowerarm_twist_02_r {parent=lowerarm_r, loc=<-9.6199 0 0>, rot=<0 0 0 -1>}
 lowerarm_twist_01_r {parent=lowerarm_r, loc=<-17.8549 0 0>, rot=<0 0 0 -1>}
 deform_elbow_in_base_r {parent=lowerarm_r, loc=<-0.2114 -5.6449 -0.827>, rot=<0 0 0 -1>}
 deform_elbow_in_r {parent=deform_elbow_in_base_r, loc=<0 0 0>, rot=<0 0 0 1>}
 deform_elbow_out_base_r {parent=lowerarm_r, loc=<-0.0001 7.5129 0.0002>, rot=<0 0 0 -1>}
 deform_elbow_out_r {parent=deform_elbow_out_base_r, loc=<0 0 0>, rot=<0 0 0 1>}
 upperarm_twist_01_r {parent=upperarm_r, loc=<-7.9796 -0.0843 0.6195>, rot=<0 0 0 -1>}
 deform_upperarm_bk_base_r {parent=upperarm_twist_01_r, loc=<6.8985 13.3826 1.3257>, rot=<0 0 0 1>}
 deform_upperarm_bk_r {parent=deform_upperarm_bk_base_r, loc=<0 0 0>, rot=<0 0 0 1>}
 deform_upperarm_fr_base_r {parent=upperarm_twist_01_r, loc=<8.3485 -10.7117 2.1456>, rot=<0 0 0 1>}
 deform_upperarm_fr_r {parent=deform_upperarm_fr_base_r, loc=<0 0 0>, rot=<0 0 0 1>}
 deform_upperarm_up_base_r {parent=upperarm_twist_01_r, loc=<11.2015 1.2234 -8.7788>, rot=<0 0 0 1>}
 deform_upperarm_up_r {parent=deform_upperarm_up_base_r, loc=<0 0 0>, rot=<0 0 0 1>}
 upperarm_twist_02_r {parent=upperarm_r, loc=<-19.5059 -0.0843 0.6195>, rot=<0 0 0 -1>}
 deform_bicep_base_r {parent=upperarm_twist_02_r, loc=<0.0002 -8.0494 -0.0003>, rot=<0 0 0 1>}
 deform_bicep_r {parent=deform_bicep_base_r, loc=<0 0 0>, rot=<0 0 0 1>}
 deform_pec_aim_base_l {parent=spine_04, loc=<5.0232 13.6719 -2.6508>, rot=<0.0672 -0.5201 0.3228 -0.7879>}
 deform_pec_aim_l {parent=deform_pec_aim_base_l, loc=<0 0 0>, rot=<0 0 0 1>}
 deform_pec_base_l {parent=deform_pec_aim_l, loc=<16.3279 0 0>, rot=<0 0 0 1>}
 deform_pec_l {parent=deform_pec_base_l, loc=<0 0 0>, rot=<0 0 0 1>}
 deform_pec_aim_base_r {parent=spine_04, loc=<5.023 13.6718 2.6508>, rot=<-0.5201 -0.0672 0.7879 0.3228>}
 deform_pec_aim_r {parent=deform_pec_aim_base_r, loc=<0 0 0>, rot=<0 0 0 1>}
 deform_pec_base_r {parent=deform_pec_aim_r, loc=<-16.3279 0 0>, rot=<0 0 0 -1>}
 deform_pec_r {parent=deform_pec_base_r, loc=<0 0 0.0002>, rot=<0 0 0 -1>}
 deform_lat_l_base {parent=spine_03, loc=<2.5566 -10.3716 -17.6918>, rot=<0 0 0 1>}
 deform_lat_l {parent=deform_lat_l_base, loc=<0 0 0>, rot=<0 0 0 1>}
 deform_lat_r_base {parent=spine_03, loc=<2.5568 -10.3716 17.6918>, rot=<0 0 1 0>}
 deform_lat_r {parent=deform_lat_r_base, loc=<0 0 0>, rot=<0 0 0 -1>}
 thigh_l {parent=pelvis, loc=<-1.8716 -0.0109 -11.9936>, rot=<0.0012 0.0572 0.0155 -0.9982>}
 calf_l {parent=thigh_l, loc=<-41.9178 -0.2246 0.0926>, rot=<0.0006 -0.0161 -0.0391 -0.9991>}
 foot_l {parent=calf_l, loc=<-38.7057 1.2438 0.2452>, rot=<0.0975 -0.0474 0.0317 -0.9936>}
 ball_l {parent=foot_l, loc=<-6.1614 13.1469 0.6807>, rot=<0.0001 -0.0001 -0.7063 -0.7079>}
 calf_twist_02_l {parent=calf_l, loc=<-15.9268 -0.0967 0.0397>, rot=<0.0237 0.0095 -0.0035 -0.9997>}
 calf_twist_01_l {parent=calf_l, loc=<-32.9335 -0.0967 0.0397>, rot=<0.0237 0.0095 -0.0035 -0.9997>}
 deform_knee_in_base_l {parent=calf_l, loc=<1.298 -8.0356 -0.0496>, rot=<0 0 0 -1>}
 deform_knee_in_l {parent=deform_knee_in_base_l, loc=<0 0 0>, rot=<0 0 0 1>}
 deform_knee_out_base_l {parent=calf_l, loc=<0 6.8739 0>, rot=<0 0 0 -1>}
 deform_knee_out_l {parent=deform_knee_out_base_l, loc=<0 0 0>, rot=<0 0 0 1>}
 thigh_twist_01_l {parent=thigh_l, loc=<-21.6303 0 0>, rot=<0.0013 0 0 -1>}
 deform_groin_base_l {parent=thigh_l, loc=<-2.1908 8.5661 -1.7442>, rot=<0 0 0 -1>}
 deform_groin_l {parent=deform_groin_base_l, loc=<0 0 0>, rot=<0 0 0 1>}
 dyn_skirt_side_l_base {parent=thigh_l, loc=<0.2356 4.0917 -7.7501>, rot=<-0.0196 -0.9831 -0.1518 -0.1002>}
 dyn_skirt_side_l_1 {parent=dyn_skirt_side_l_base, loc=<0 0 0>, rot=<0 0 0 1>}
 dyn_skirt_side_l_2 {parent=dyn_skirt_side_l_1, loc=<10.7141 0 0>, rot=<0 0 0 1>}
 thigh_r {parent=pelvis, loc=<-1.8716 -0.0109 11.9936>, rot=<0.0572 -0.0012 0.9982 0.0155>}
 calf_r {parent=thigh_r, loc=<41.9178 0.2246 -0.0926>, rot=<0.0006 -0.0161 -0.0391 -0.9991>}
 foot_r {parent=calf_r, loc=<38.7057 -1.2438 -0.2452>, rot=<0.0975 -0.0474 0.0317 -0.9936>}
 ball_r {parent=foot_r, loc=<6.1614 -13.1469 -0.6807>, rot=<0.0001 -0.0001 -0.7063 -0.7079>}
 calf_twist_02_r {parent=calf_r, loc=<15.9268 0.0967 -0.0397>, rot=<0.0237 0.0095 -0.0035 -0.9997>}
 calf_twist_01_r {parent=calf_r, loc=<32.9335 0.0967 -0.0397>, rot=<0.0237 0.0095 -0.0035 -0.9997>}
 deform_knee_in_base_r {parent=calf_r, loc=<-1.2981 8.0356 0.0496>, rot=<0 0 0 -1>}
 deform_knee_in_r {parent=deform_knee_in_base_r, loc=<0 0 0>, rot=<0 0 0 1>}
 deform_knee_out_base_r {parent=calf_r, loc=<0 -6.8739 0>, rot=<0 0 0 -1>}
 deform_knee_out_r {parent=deform_knee_out_base_r, loc=<0 0 0>, rot=<0 0 0 1>}
 thigh_twist_01_r {parent=thigh_r, loc=<21.6303 0 0>, rot=<0.0013 0 0 -1>}
 deform_groin_base_r {parent=thigh_r, loc=<2.1908 -8.5661 1.7442>, rot=<0 0 0 -1>}
 deform_groin_r {parent=deform_groin_base_r, loc=<0 0 0>, rot=<0 0 0 -1>}
 dyn_skirt_side_r_base {parent=thigh_r, loc=<-0.0471 -4.3054 8.4183>, rot=<-0.0196 -0.9831 -0.1518 -0.1002>}
 dyn_skirt_side_r_1 {parent=dyn_skirt_side_r_base, loc=<0 0 0>, rot=<-0.0006 0 -0.0001 -1>}
 dyn_skirt_side_r_2 {parent=dyn_skirt_side_r_1, loc=<-10.6446 0 0.3432>, rot=<0 0 0 -1>}
 deform_glute_base_l {parent=pelvis, loc=<-10.1877 -11.6217 -12.899>, rot=<0.0275 0.0501 -0.4472 -0.8926>}
 deform_glute_l {parent=deform_glute_base_l, loc=<0 0 0>, rot=<0 0 0 1>}
 deform_glute_base_r {parent=pelvis, loc=<-10.1877 -11.6218 12.899>, rot=<0.0501 -0.0275 0.8926 -0.4472>}
 deform_glute_r {parent=deform_glute_base_r, loc=<0 0 0>, rot=<0 0 0 1>}
 dyn_skirt_frnt_base {parent=pelvis, loc=<-9.8856 13.6957 0>, rot=<-0.043 -0.9991 0 0>}
 dyn_skirt_frnt_1 {parent=dyn_skirt_frnt_base, loc=<0 0 0>, rot=<0 0 0 1>}
 dyn_skirt_frnt_2 {parent=dyn_skirt_frnt_1, loc=<15.4156 0 0>, rot=<0 0 0 1>}
 dyn_back_r_base {parent=pelvis, loc=<1.2158 -11.4468 8.6019>, rot=<0.1944 -0.972 0.1279 0.0314>}
 dyn_back_r_1 {parent=dyn_back_r_base, loc=<8.3636 0 0>, rot=<0 0 -0.1589 -0.9873>}
 dyn_back_r_2 {parent=dyn_back_r_1, loc=<9.7368 0 0>, rot=<0 0 -0.0558 -0.9984>}
 dyn_back_r_3 {parent=dyn_back_r_2, loc=<11.4225 0 0>, rot=<0 0 0.002 -1>}
 dyn_back_l_base {parent=pelvis, loc=<1.06 -12.5318 -3.4296>, rot=<0.2198 -0.9723 -0.0776 -0.0191>}
 dyn_back_l_1 {parent=dyn_back_l_base, loc=<7.8774 0 0>, rot=<0 0 -0.1948 -0.9809>}
 dyn_back_l_2 {parent=dyn_back_l_1, loc=<8.8726 -0.2079 0>, rot=<0 0 -0.0366 -0.9993>}
 dyn_back_l_3 {parent=dyn_back_l_2, loc=<13.0383 0 0>, rot=<0 0 0.0077 -1>}
 dyn_skirt_frnt_side_l_base {parent=pelvis, loc=<-7.3026 8.6206 -15.236>, rot=<-0.0929 -0.9245 0.3583 -0.0914>}
 dyn_skirt_frnt_side_l_1 {parent=dyn_skirt_frnt_side_l_base, loc=<0 0 0>, rot=<0 0 0 1>}
 dyn_skirt_frnt_side_r_base {parent=pelvis, loc=<-7.2197 8.6114 15.2158>, rot=<0.9245 -0.0929 -0.0914 -0.3583>}
 dyn_skirt_frnt_side_r_1 {parent=dyn_skirt_frnt_side_r_base, loc=<0 0 0>, rot=<0 0 0 1>}
 glute_support_ALT_base {parent=pelvis, loc=<-21.161 -15.9201 0>, rot=<0 0 0.7197 -0.6943>}
 glute_support_ALT_1 {parent=glute_support_ALT_base, loc=<0 0 0>, rot=<0 0 0 1>}
 Root {parent=Root, type=Socket, loc=<0 0 0>, rot=<0 0 0 1>}
]
variables [
 Clavicle_L_raise {type=double, default=0}
 Forward Movement {type=double, default=0}
 Is Backward Forward MOvement {type=double, default=0}
 Is Baseball Sliding {type=double, default=0}
 Is Crouching {type=double, default=0}
 Is DBNO {type=double, default=0}
 Is Parachute Open {type=double, default=0}
 Is Right Left Movement {type=double, default=0}
 Is Skydive Float {type=double, default=0}
 Right Movement {type=double, default=0}
]
