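# Physics asset for the DenseFog large male player body, defining collision and joint limits for its skirt bones.
# Five chains (front, both sides, back left and right) run from zero-size base bodies through capsules, with boxes ending the front and side panels.
# Tight swing and twist limits link each chain.
# Pelvis, thigh, calf and large root-sphere bodies add body shapes; only right-leg joints and right-side skirt base have free-motion constraints.
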PhysicsAsset /BRCosmetics/Characters/Player/Male/Large/Bodies/M_LRG_DenseFog/Meshes/M_LRG_DenseFog_Physics_Skirt
bodies {
 calf_l [
  {shapeType=Sphere, radius=12.0797, shapeName=calf_l_sphere, center=<-14.6794 -3.8014 0.3937>}
  {shapeType=Capsule, length=42.8565, radius=8.1111, center=<-21.1773 -1.7178 0.7233>, rotation=<rot 90 0 0>}
 ]
 calf_r [
  {shapeType=Sphere, radius=11.5993, shapeName=calf_r_sphere, center=<13.4424 4.3743 0>}
  {shapeType=Capsule, length=37.0819, radius=8.2514, center=<20.2465 2.681 -1.122>, rotation=<rot 90 0 0>}
 ]
 dyn_back_l_1 [{shapeType=Capsule, length=10.0786, radius=4, shapeName=dyn_back_l_1_capsule, center=<6.5601 -4 0>}]
 dyn_back_l_2 [{shapeType=Capsule, length=8.2566, radius=6.6046, shapeName=dyn_back_l_2_capsule, center=<6.9822 -3.8967 0>}]
 dyn_back_l_3 [{shapeType=Capsule, length=9.0961, radius=7.9241, shapeName=dyn_back_l_3_capsule, center=<10.4881 -3.4265 0>, rotation=<rot 90 0 0>}]
 dyn_back_l_base [{shapeType=Capsule, length=0, radius=0, shapeName=dyn_back_l_base_capsule}]
 dyn_back_r_1 [{shapeType=Capsule, length=5.9895, radius=4, shapeName=dyn_back_r_1_capsule, center=<6.2234 -4.0393 0>}]
 dyn_back_r_2 [{shapeType=Capsule, length=6.6935, radius=5.9175, shapeName=dyn_back_r_2_capsule, center=<4.8812 -3.7983 0>}]
 dyn_back_r_3 [{shapeType=Capsule, length=8.931, radius=6.1557, shapeName=dyn_back_r_3_capsule, center=<7.9503 -4.0858 0>, rotation=<rot -90 180 180>}]
 dyn_back_r_base [{shapeType=Capsule, length=0, radius=0, shapeName=dyn_back_r_base_capsule}]
 dyn_skirt_frnt_1 [{shapeType=Capsule, length=17.0344, radius=6, shapeName=dyn_skirt_frnt_1_capsule, center=<5.4509 4.9023 0>}]
 dyn_skirt_frnt_2 [
  {shapeType=Capsule, length=0, radius=0, shapeName=dyn_skirt_frnt_2_capsule}
  {
   shapeType=Box
   extentX=19.5006
   extentY=5.2543
   extentZ=46.4888
   shapeName=dyn_skirt_frnt_2_box
   center=<5.6094 0.4855 0>
   rotation=<rot 0 -3.2662 0>
  }
 ]
 dyn_skirt_frnt_base [{shapeType=Capsule, length=0, radius=0, shapeName=dyn_skirt_frnt_base_capsule}]
 dyn_skirt_side_l_1 [{shapeType=Capsule, length=5.8611, radius=5.1389, shapeName=dyn_skirt_side_l_1_capsule, center=<5.091 0 3.856>, rotation=<rot 0 0 90>}]
 dyn_skirt_side_l_2 [
  {shapeType=Capsule, length=0, radius=0, shapeName=dyn_skirt_side_l_2_capsule}
  {
   shapeType=Box
   extentX=25.483
   extentY=19.144
   extentZ=4.98
   shapeName=dyn_skirt_side_l_2_box
   center=<9.8144 0 -1.1441>
   rotation=<rot -7.3911 0.0007 -0.0002>
  }
 ]
 dyn_skirt_side_l_base [{shapeType=Capsule, length=0, radius=0, shapeName=dyn_skirt_side_l_base_capsule}]
 dyn_skirt_side_r_1 [{shapeType=Capsule, length=5.6487, radius=5.3513, shapeName=dyn_skirt_side_r_1_capsule, center=<-5.5907 0 -3.1861>, rotation=<rot 0 0 90>}]
 dyn_skirt_side_r_2 [
  {shapeType=Capsule, length=0, radius=0, shapeName=dyn_skirt_side_r_2_capsule}
  {
   shapeType=Box
   extentX=22.9705
   extentY=16.2109
   extentZ=4.9322
   shapeName=dyn_skirt_side_r_2_box
   center=<-10.5984 0 1.9417>
   rotation=<rot -7.9449 0.0003 -0.0001>
  }
 ]
 dyn_skirt_side_r_base [{shapeType=Capsule, length=0, radius=0, shapeName=dyn_skirt_side_r_base_capsule}]
 pelvis [{shapeType=Capsule, length=15.6203, radius=12.5895, center=<-12.2501 -1.3768 0>}]
 root [{shapeType=Sphere, radius=116.4129, shapeName=root_sphere, center=<0 -2.1036 -106.3735>}]
 thigh_l [{shapeType=Capsule, length=37.5935, radius=11.0959, center=<-20.8769 -0.0631 1.4343>, rotation=<rot 86.81 179.9868 0>}]
 thigh_r [{shapeType=Capsule, length=37.5935, radius=11.0959, center=<20.8769 0.0631 -1.4343>, rotation=<rot 86.81 179.9868 180>}]
}
constraints [
 {bone1Name=dyn_back_l_1, bone2Name=dyn_back_l_base, swing1LimitDegrees=10, swing2LimitDegrees=5, twistLimitDegrees=0}
 {bone1Name=dyn_back_l_2, bone2Name=dyn_back_l_1, swing1LimitDegrees=10, swing2LimitDegrees=10, twistLimitDegrees=5}
 {bone1Name=dyn_back_l_3, bone2Name=dyn_back_l_2, swing1LimitDegrees=20, swing2LimitDegrees=15, twistLimitDegrees=5}
 {bone1Name=dyn_back_r_1, bone2Name=dyn_back_r_base, swing1LimitDegrees=10, swing2LimitDegrees=5, twistLimitDegrees=0}
 {bone1Name=dyn_back_r_2, bone2Name=dyn_back_r_1, swing1LimitDegrees=10, swing2LimitDegrees=10, twistLimitDegrees=5}
 {bone1Name=dyn_back_r_3, bone2Name=dyn_back_r_2, swing1LimitDegrees=20, swing2LimitDegrees=10, twistLimitDegrees=5}
 {bone1Name=dyn_skirt_frnt_2, bone2Name=dyn_skirt_frnt_1, swing1LimitDegrees=10, swing2LimitDegrees=2, twistLimitDegrees=15}
 {bone1Name=dyn_skirt_frnt_1, bone2Name=dyn_skirt_frnt_base, swing1LimitDegrees=20, swing2LimitDegrees=5, twistLimitDegrees=15}
 {bone1Name=dyn_skirt_side_l_1, bone2Name=dyn_skirt_side_l_base, swing1LimitDegrees=5, swing2LimitDegrees=5, twistLimitDegrees=10}
 {bone1Name=dyn_skirt_side_l_2, bone2Name=dyn_skirt_side_l_1, swing1LimitDegrees=30, swing2LimitDegrees=5, twistLimitDegrees=10}
 {bone1Name=dyn_skirt_side_r_1, bone2Name=dyn_skirt_side_r_base, swing1LimitDegrees=5, swing2LimitDegrees=5, twistLimitDegrees=10}
 {bone1Name=dyn_skirt_side_r_2, bone2Name=dyn_skirt_side_r_1, swing1LimitDegrees=2, swing2LimitDegrees=5, twistLimitDegrees=10}
 {
  bone1Name=calf_r
  bone2Name=thigh_r
  swing1LimitDegrees=45
  swing1Motion=Free
  swing2LimitDegrees=45
  swing2Motion=Free
  twistLimitDegrees=45
  twistMotion=Free
 }
 {
  bone1Name=dyn_skirt_side_r_base
  bone2Name=thigh_r
  swing1LimitDegrees=45
  swing1Motion=Free
  swing2LimitDegrees=45
  swing2Motion=Free
  twistLimitDegrees=45
  twistMotion=Free
 }
 {
  bone1Name=thigh_r
  bone2Name=pelvis
  swing1LimitDegrees=45
  swing1Motion=Free
  swing2LimitDegrees=45
  swing2Motion=Free
  twistLimitDegrees=45
  twistMotion=Free
 }
]
