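# Animation Blueprint for the M_LRG_DenseFog player body cosmetic, adding secondary motion to the pose copied from its attached parent mesh.
# A Control Rig adjusts that pose using movement and state floats refreshed each update.
# A pelvis-space RigidBody node then simulates the skirt physics asset with a computed gravity override.
# In the front end, an idle-bypass curve blends DefaultSlot animation over the skirt, coif, helmet and shoulderpad bones.
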AnimBlueprint /BRCosmetics/Characters/Player/Male/Large/Bodies/M_LRG_DenseFog/Meshes/M_LRG_DenseFog_AnimBP
target_skeleton=&/M_LRG_DenseFog_Skeleton
graphs [
 AnimGraph {nodes [
   n1 = Control Rig!(
    Source=Copy Pose From Mesh() {bUseAttachedParent=true}
    Alpha
    Is Crouching:double=Get Is Crouching() -> (Is Crouching:double)
    Is Skydive Float:double=Get Is Skydive Float() -> (Is Skydive Float:double)
    Is Baseball Sliding:double=Get Is Baseball Sliding() -> (Is Baseball Sliding:double)
    Is Parachute Open:double=Get Is Parachute Open() -> (Is Parachute Open:double)
    Is DBNO:double=Get Is DBNO() -> (Is DBNO:double)
    Is Right Left Movement:double=Get Is Right Left Movement() -> (Is Right Left Movement:double)
    Is Backward Forward MOvement:double=Get Is Forward Backward Movement() -> (Is Forward Backward Movement:double)
   ) -> (Pose) {
    alphaBoolBlend {blendOption=Linear}
    controlRigAssetReference {blueprintRigClass=&/M_LRG_DenseFog_CtrlRig.M_LRG_DenseFog_CtrlRig_C}
   }
   n2 = RigidBody(ComponentPose=Local To Component(LocalPose=@n1), ExternalForce=Get Gravity Override()) {
    bClampLinearTranslationLimitToRefPose=true
    bOverrideWorldGravity=true
    overridePhysicsAsset=&/M_LRG_DenseFog_Physics_Skirt
    simulationSpace=BaseBoneSpace
    alphaBoolBlend {blendOption=Linear}
    baseBoneRef {boneName=pelvis}
    componentAppliedLinearAccClamp=<3000 3000 3000>
    componentLinearAccScale=<1.3 1.3 1.3>
    componentLinearVelScale=<1.3 1.3 1.3>
    overrideWorldGravity {z=-980}
    simSpaceSettings {
     maxAngularAcceleration=500
     maxAngularVelocity=4
     maxLinearAcceleration=400
     maxLinearVelocity=150
     velocityScaleZ=0
     worldAlpha=1
     externalLinearDragV=<0.1 0.1>
    }
   }
   n3 = Layered blend per bone(
    BasePose=Component To Local(ComponentPose=@n2)
    BlendPoses_0=Slot 'DefaultSlot'()
    BlendWeights_0=Get Custome Idle is Playing Float() -> (Custome Idle is Playing Float:float)
   ) {layerSetup [{branchFilters [
       {boneName=shoulderpad_02_l_base}
       {boneName=shoulderpad_01_l}
       {boneName=shoulderpad_01_l_base}
       {boneName=deform_shoulderpad_l}
       {boneName=deform_shoulderpad_l_base}
       {boneName=dyn_coif_side_r_2}
       {boneName=dyn_coif_side_r_1}
       {boneName=dyn_coif_side_r_base}
       {boneName=dyn_coif_side_l_2}
       {boneName=dyn_coif_side_l_1}
       {boneName=dyn_coif_side_l_base}
       {boneName=dyn_coif_back_2}
       {boneName=dyn_coif_back_1}
       {boneName=dyn_coif_back_base}
       {boneName=dyn_helmet_mid_1_1}
       {boneName=dyn_helmet_mid_1_base}
       {boneName=dyn_helmet_back_1}
       {boneName=dyn_helmet_back_base}
       {boneName=shoulderpad_02_l}
       {boneName=dyn_skirt_frnt_base}
       {boneName=dyn_skirt_frnt_1}
       {boneName=dyn_skirt_frnt_2}
       {boneName=dyn_skirt_side_r_base}
       {boneName=dyn_skirt_side_r_1}
       {boneName=dyn_skirt_side_r_2}
       {boneName=dyn_skirt_side_l_base}
       {boneName=dyn_skirt_side_l_1}
       {boneName=dyn_skirt_side_l_2}
      ]}]}
   Output Pose(Result=@n3) {layerGroup=None}
  ]}
 EventGraph {nodes [
   Event BlueprintUpdateAnimation()
   > n1 = Sequence!(execute) -> (then_0, then_1, then_2:Exec)
   n2 = Get bIsCrouching()
   n3 = Get bIsCrouchSprinting()
   > Set Is Crouching(Is Crouching:double="To Float (Boolean)"(InBool=OR Boolean(A=@n3, B=@n2))) -> (Output_Get:double)
   > Set Is Skydive Float(Is Skydive Float:double="To Float (Boolean)"(InBool=Get bIsSkydiveFloating())) -> (Output_Get:double)
   > Set Is Baseball Sliding(Is Baseball Sliding:double="To Float (Boolean)"(InBool=Get bIsBaseballSliding())) -> (Output_Get:double)
   > Set Is Parachute Open(Is Parachute Open:double="To Float (Boolean)"(InBool=Get bIsParachuteOpened())) -> (Output_Get:double)
   > Set Is DBNO(Is DBNO:double="To Float (Boolean)"(InBool=Get bIsDBNO())) -> (Output_Get:double)
   Branch(execute=@n1.then_2, Condition=Get bIsInFrontEnd())
   > Set Custome Idle is Playing Float(Custome Idle is Playing Float:float=GetCurveValue(self=GetOwnerFrontEndAnimInstance(), CurveName=Frontend_Idle_Bypass)) -> (Output_Get:float)
   n4 = Gravity Override State Function DEPRECATED(
    execute=@n1.then_1
    CharacterPartAnimInstance=Self-Reference()
    GameTimeInSeconds=GetGameTimeInSeconds()
    GravityOverrideParams=Get Gravity Override Params 0() -> (Gravity Override Params 0:Gravity Override Params Struct Structure)
   )
   > Set Gravity Override(Gravity Override=@n4)
   > n5 = PlayerMovement(AnimInstance=Self-Reference(), XDotProductCoefficient=1, YDotProductCoefficient=1, ZDotProductCoefficient=1)
   > Set Is Right Left Movement(Is Right Left Movement:double=@n5) -> (Output_Get:double)
   > Set Is Forward Backward Movement(Is Forward Backward Movement:double=@n5.YMovement) -> (Output_Get:double)
  ]}
]
variables [
 Custome Idle is Playing Float {type=number}
 Gravity Override {type=Vector}
 Gravity Override Params 0 {type=GravityOverrideParamsStruct}
 Is Baseball Sliding {type=number}
 Is Crouching {type=number}
 Is DBNO {type=number}
 Is Forward Backward Movement {type=number}
 Is Parachute Open {type=number}
 Is Right Left Movement {type=number}
 Is Skydive Float {type=number}
]
