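# Control Rig for the M_LRG_DenseFog face accessory that poses helmet and coif dyn base bones as the head moves.
# Spherical pose readers on the head weight transform offsets, with separate branches for crouching and non-crouching, each including a reader scaled by Is Rifle.
# Most reader weights are smoothed with AlphaInterp.
# The non-crouching branch also tilts the helmet back bone during skydive floating, weighted by head rotation relative to the initial pelvis.
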ControlRigBlueprint /BRCosmetics/Characters/Player/Male/Large/Bodies/M_LRG_DenseFog/Meshes/Parts/M_LRG_DenseFog_FaceAcc_CtrlRig
graphs [
 RigVMFunctionLibrary {links [], nodes []}
 RigVMModel {nodes [
   RigUnit_BeginExecution = &RigUnit_BeginExecution()
   DISPATCH_RigVMDispatch_CoreEquals = RigVMDispatchNode(B=1, A=RigVMVariableNode() {variable=Is Crouching}.Value) {template="DISPATCH_RigVMDispatch_CoreEquals(in A,in B,out Result)"}
   RigVMFunction_ControlFlowBranch = &RigVMFunction_ControlFlowBranch(Condition=@DISPATCH_RigVMDispatch_CoreEquals.Result, ExecuteContext=@RigUnit_BeginExecution.ExecutePin)
   n1 = &RigUnit_SphericalPoseReader(
    ActiveRegionScaleFactors=(PositiveWidth=0.7,NegativeWidth=1,PositiveHeight=1,NegativeHeight=0.7)
    ActiveRegionSize=0.1
    Debug=(bDrawDebug=True,bDraw2D=False,bDrawLocalAxes=False,DebugScale=25,DebugSegments=20,DebugThickness=0.25)
    DriverAxis=(X=1,Y=0,Z=0)
    DriverItem=(Type=Bone,Name="head")
    FalloffRegionScaleFactors=(PositiveWidth=0.5,NegativeWidth=1,PositiveHeight=1,NegativeHeight=0.7)
    FalloffSize=0.2
    FlipHeightScaling=False
    FlipWidthScaling=False
    OptionalParentItem=(Type=Bone,Name="None")
    RotationOffset=(X=-70,Y=0,Z=60)
    ExecutePin=@RigVMFunction_ControlFlowBranch.False
   )
   n2 = &RigVMFunction_AlphaInterp(
    InterpSpeedDecreasing=12
    InterpSpeedIncreasing=12
    OutRange=(Max=1)
    Scale=1
    bClampResult=true
    bMapRange=true
    Bias=~
    ClampMin=~
    Value=@n1.OutputParam
   )
   > &RigUnit_ModifyTransforms(
    ItemToModify="((Item=(Type=Bone,Name=\"dyn_helmet_back_base\"),Transform=(Rotation=(X=0,Y=-0.5225,Z=0,W=0.8526),Translation=(X=-2,Y=0,Z=0),Scale3D=(X=1,Y=1,Z=1))),(Item=(Type=Bone,Name=\"dyn_coif_side_l_base\"),Transform=(Rotation=(X=0,Y=0,Z=0,W=1),Translation=(X=0,Y=0,Z=0) ...(truncated)"
    WeightMinimum=0
    Weight=@n2.Result
   )
   RigVMFunction_MathQuaternionToEuler = &RigVMFunction_MathQuaternionToEuler(
    RotationOrder=ZYX
    Value=&RigUnit_GetRelativeTransformForItem(Child=(Type=Bone,Name="head"), Parent=(Type=Bone,Name="pelvis"), bChildInitial=false, bParentInitial=true).RelativeTransform.Rotation
   )
   > &RigUnit_ModifyTransforms(
    ItemToModify=((Item=(Type=Bone,Name="dyn_helmet_back_base"),Transform=(Rotation=(X=0,Y=-0.4617,Z=0,W=0.887),Translation=(X=0,Y=0,Z=0),Scale3D=(X=1,Y=1,Z=1))))
    WeightMinimum=0
    Weight=&RigVMFunction_MathFloatMul(
     B=&RigVMFunction_MathFloatRemap(SourceMaximum=15, SourceMinimum=10, TargetMinimum=0, bClamp=true, Value=@RigVMFunction_MathQuaternionToEuler.Result.Z).Result
     A=RigVMVariableNode() {variable=Is Skydive Floating}.Value
    ) {template="Multiply::Execute(in A,in B,out Result)"}.Result
   )
   > n3 = &RigUnit_SphericalPoseReader(
    ActiveRegionScaleFactors=(PositiveWidth=1,NegativeWidth=1,PositiveHeight=1,NegativeHeight=0.7)
    ActiveRegionSize=0.1
    Debug=(bDrawDebug=false,bDraw2D=False,bDrawLocalAxes=False,DebugScale=25,DebugSegments=20,DebugThickness=0.25)
    DriverAxis=(X=1,Y=0,Z=0)
    DriverItem=(Type=Bone,Name="head")
    FalloffRegionScaleFactors=(PositiveWidth=1,NegativeWidth=1,PositiveHeight=1,NegativeHeight=0.7)
    FalloffSize=0.2
    FlipHeightScaling=False
    FlipWidthScaling=False
    OptionalParentItem=(Type=Bone,Name="None")
    RotationOffset=(X=-70,Y=0,Z=70)
   )
   Multiply = &RigVMFunction_MathDoubleMul(A=@n3.OutputParam, B=RigVMVariableNode() {variable=Is Rifle}.Value)
   > &RigUnit_ModifyTransforms(
    ItemToModify="((Item=(Type=Bone,Name=\"dyn_coif_side_r_base\"),Transform=(Rotation=(X=-0.0594,Y=0.1632,Z=0.3368,W=0.9254),Translation=(X=0,Y=0,Z=0),Scale3D=(X=1,Y=1,Z=1))),(Item=(Type=Bone,Name=\"dyn_coif_side_l_base\"),Transform=(Rotation=(X=0.1736,Y=0,Z=0,W=0.9848),Transl ...(truncated)"
    WeightMinimum=0
    Weight=@Multiply.Result
   )
   n4 = &RigUnit_SphericalPoseReader(
    ActiveRegionScaleFactors=(PositiveWidth=0.4,NegativeWidth=0.5,PositiveHeight=0.5,NegativeHeight=0.5)
    ActiveRegionSize=0.1
    Debug=(bDrawDebug=True,bDraw2D=False,bDrawLocalAxes=False,DebugScale=25,DebugSegments=20,DebugThickness=0.25)
    DriverAxis=(X=1,Y=0,Z=0)
    DriverItem=(Type=Bone,Name="head")
    FalloffRegionScaleFactors=(PositiveWidth=0.2,NegativeWidth=0.6,PositiveHeight=0.6,NegativeHeight=0.2)
    FalloffSize=0.2
    FlipHeightScaling=False
    FlipWidthScaling=False
    OptionalParentItem=(Type=Bone,Name="None")
    RotationOffset=(X=-75,Y=0,Z=60)
    ExecutePin=@RigVMFunction_ControlFlowBranch.True
   )
   n5 = &RigVMFunction_AlphaInterp(
    InterpSpeedDecreasing=12
    InterpSpeedIncreasing=12
    OutRange=(Max=1)
    Scale=1
    bClampResult=true
    bMapRange=true
    Bias=~
    ClampMin=~
    Value=@n4.OutputParam
   )
   > &RigUnit_ModifyTransforms(
    ItemToModify="((Item=(Type=Bone,Name=\"dyn_helmet_back_base\"),Transform=(Rotation=(X=0,Y=-0.5736,Z=0,W=0.8192),Translation=(X=-2,Y=0,Z=0),Scale3D=(X=1,Y=1,Z=1))),(Item=(Type=Bone,Name=\"dyn_coif_side_r_base\"),Transform=(Rotation=(X=0.1653,Y=0.0998,Z=-0.5399,W=0.8193),Tran ...(truncated)"
    WeightMinimum=0
    Weight=@n5.Result
   )
   > n6 = &RigUnit_SphericalPoseReader(
    ActiveRegionScaleFactors=(PositiveWidth=1,NegativeWidth=0.5,PositiveHeight=0.5,NegativeHeight=0.5)
    ActiveRegionSize=0.1
    Debug=(bDrawDebug=true,bDraw2D=False,bDrawLocalAxes=False,DebugScale=25,DebugSegments=20,DebugThickness=0.25)
    DriverAxis=(X=1,Y=0,Z=0)
    DriverItem=(Type=Bone,Name="head")
    FalloffRegionScaleFactors=(PositiveWidth=1,NegativeWidth=0.6,PositiveHeight=0.6,NegativeHeight=0.6)
    FalloffSize=0.2
    FlipHeightScaling=False
    FlipWidthScaling=False
    OptionalParentItem=(Type=Bone,Name="None")
    RotationOffset=(X=-75,Y=0,Z=60)
   )
   n7 = &RigVMFunction_AlphaInterp(
    InterpSpeedDecreasing=12
    InterpSpeedIncreasing=12
    OutRange=(Max=1)
    Scale=1
    bClampResult=true
    bMapRange=true
    Bias=~
    ClampMin=~
    Value=&RigVMFunction_MathDoubleMul(A=@n6.OutputParam, B=RigVMVariableNode() {variable=Is Rifle}.Value).Result
   )
   > &RigUnit_ModifyTransforms(
    ItemToModify="((Item=(Type=Bone,Name=\"dyn_coif_side_r_base\"),Transform=(Rotation=(X=-0.0545,Y=0.1853,Z=-0.1555,W=0.9688),Translation=(X=2,Y=0,Z=0),Scale3D=(X=1,Y=1,Z=1))),(Item=(Type=Bone,Name=\"dyn_helmet_back_base\"),Transform=(Rotation=(X=0,Y=0,Z=0,W=1),Translation=(X= ...(truncated)"
    WeightMinimum=0
    Weight=@n7.Result
   )
  ]}
]
hierarchy [
 root {loc=<0 0 0>, rot=<0 0 0 1>}
 pelvis {parent=root, loc=<0 -0.7321 92.4218>, rot=<0.0127 0.707 -0.0127 -0.707>}
 spine_01 {parent=pelvis, loc=<5.6002 -0.5128 0>, rot=<0 0 0.034 -0.9994>}
 spine_02 {parent=spine_01, loc=<8.0753 0.0139 0>, rot=<0 0 0.0238 -0.9997>}
 spine_03 {parent=spine_02, loc=<7.0393 0 0>, rot=<0 0 0.0149 -0.9999>}
 spine_04 {parent=spine_03, loc=<8.1024 0 0>, rot=<0 0 -0.0132 -0.9999>}
 spine_05 {parent=spine_04, loc=<10.0146 0 0>, rot=<0 0 -0.0015 -1>}
 neck_01 {parent=spine_05, loc=<10.815 1.9196 0>, rot=<0 0 -0.1864 -0.9825>}
 neck_02 {parent=neck_01, loc=<4.7347 0.3045 0>, rot=<0 0 0.0363 -0.9993>}
 head {parent=neck_02, loc=<5.7632 0.5552 0>, rot=<0 0 0.115 -0.9934>}
 dyn_helmet_back_base {parent=head, loc=<9.9656 -8.2161 0>, rot=<0.5021 -0.4979 0.4979 -0.5021>}
 dyn_helmet_back_1 {parent=dyn_helmet_back_base, loc=<0 0 0>, rot=<0 0 0 1>}
 dyn_helmet_frnt_base {parent=head, loc=<16.2014 7.8561 0>, rot=<0.6214 -0.3374 0.3374 -0.6214>}
 dyn_helmet_frnt_1 {parent=dyn_helmet_frnt_base, loc=<0 0 0>, rot=<0 0 0 1>}
 dyn_helmet_mid_1_base {parent=head, loc=<15.0136 -4.4723 0>, rot=<0.6788 -0.198 0.198 -0.6788>}
 dyn_helmet_mid_1_1 {parent=dyn_helmet_mid_1_base, loc=<0 0 0>, rot=<0 0 0 1>}
 dyn_helmet_mid_2_base {parent=head, loc=<17.5569 2.5244 0>, rot=<0.7031 -0.0754 0.0754 -0.7031>}
 dyn_helmet_mid_2_1 {parent=dyn_helmet_mid_2_base, loc=<0 0 0>, rot=<0 0 0 1>}
 dyn_coif_back_base {parent=head, loc=<4.8201 -6.6846 0>, rot=<0.269 -0.9632 0 0>}
 dyn_coif_back_1 {parent=dyn_coif_back_base, loc=<0 0 0>, rot=<0 0 0 1>}
 dyn_coif_back_2 {parent=dyn_coif_back_1, loc=<3.8769 0 0>, rot=<0 0 0.04 -0.9992>}
 dyn_coif_side_l_base {parent=head, loc=<4.9303 -4.9121 -5.3125>, rot=<0.2996 -0.8921 -0.326 -0.0896>}
 dyn_coif_side_l_1 {parent=dyn_coif_side_l_base, loc=<0 0 0>, rot=<0 0 0 1>}
 dyn_coif_side_l_2 {parent=dyn_coif_side_l_1, loc=<3.8769 0 0>, rot=<0 0 0.04 -0.9992>}
 dyn_coif_side_r_base {parent=head, loc=<4.9305 -4.9121 5.3124>, rot=<0.8921 0.2996 -0.0896 0.326>}
 dyn_coif_side_r_1 {parent=dyn_coif_side_r_base, loc=<0 0 0>, rot=<0 0 0 1>}
 dyn_coif_side_r_2 {parent=dyn_coif_side_r_1, loc=<-3.8769 0 0>, rot=<0 0 0.04 -0.9992>}
 clavicle_l {parent=spine_05, loc=<2.5856 -0.2866 -4.3777>, rot=<0.0465 -0.7289 -0.009 -0.6829>}
 upperarm_l {parent=clavicle_l, loc=<18.2264 0.1342 -0.001>, rot=<0.0253 -0.2693 0.0013 -0.9627>}
 upperarm_twist_01_l {parent=upperarm_l, loc=<7.9796 0.0843 -0.6195>, rot=<0 0 0 -1>}
 deform_upperarm_bk_base_l {parent=upperarm_twist_01_l, loc=<-6.8983 -13.3826 -1.3258>, rot=<0 0 0 1>}
 deform_upperarm_bk_l {parent=deform_upperarm_bk_base_l, loc=<0 0 0>, rot=<0 0 0 1>}
 deform_upperarm_fr_base_l {parent=upperarm_twist_01_l, loc=<-8.3487 10.7117 -2.1455>, rot=<0 0 0 1>}
 deform_upperarm_fr_l {parent=deform_upperarm_fr_base_l, loc=<0 0 0>, rot=<0 0 0 1>}
 deform_upperarm_up_base_l {parent=upperarm_twist_01_l, loc=<-11.2012 -1.2234 8.7784>, rot=<0 0 0 1>}
 deform_upperarm_up_l {parent=deform_upperarm_up_base_l, loc=<0 0 0>, rot=<0 0 0 1>}
 clavicle_r {parent=spine_05, loc=<2.5855 -0.2866 4.3777>, rot=<0.7289 0.0465 -0.6829 0.009>}
 upperarm_r {parent=clavicle_r, loc=<-18.2264 -0.1342 0.001>, rot=<0.0253 -0.2693 0.0013 -0.9627>}
 upperarm_twist_01_r {parent=upperarm_r, loc=<-7.9796 -0.0843 0.6195>, rot=<0 0 0 -1>}
 deform_upperarm_bk_base_r {parent=upperarm_twist_01_r, loc=<6.8985 13.3826 1.3257>, rot=<0 0 0 1>}
 deform_upperarm_bk_r {parent=deform_upperarm_bk_base_r, loc=<0 0 0>, rot=<0 0 0 1>}
 deform_upperarm_fr_base_r {parent=upperarm_twist_01_r, loc=<8.3485 -10.7117 2.1456>, rot=<0 0 0 1>}
 deform_upperarm_fr_r {parent=deform_upperarm_fr_base_r, loc=<0 0 0>, rot=<0 0 0 1>}
 deform_upperarm_up_base_r {parent=upperarm_twist_01_r, loc=<11.2015 1.2234 -8.7788>, rot=<0 0 0 1>}
 deform_upperarm_up_r {parent=deform_upperarm_up_base_r, loc=<0 0 0>, rot=<0 0 0 1>}
 root {parent=root, type=Socket, loc=<0 0 0>, rot=<0 0 0 1>}
]
variables [
 Is Crouching {type=double, default=0}
 Is Rifle {type=double, default=0}
 Is Skydive Floating {type=double, default=0}
]
