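# Physics asset for the face accessory of the large male DenseFog body, defining secondary motion for its coif and helmet pieces.
# Three two-segment coif chains (back, left, right) and four single-segment helmet pieces hang from small base bodies constrained to neck_01 with free rotation.
# Chain segments have narrow swing limits and zero twist limits, with boxes shaping the coif tips.
# Large neck_01 and spine_05 capsules represent the neck and upper torso.
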PhysicsAsset /BRCosmetics/Characters/Player/Male/Large/Bodies/M_LRG_DenseFog/Meshes/Parts/M_LRG_DenseFog_FaceAcc_Physics
bodies {
 dyn_coif_back_1 [{shapeType=Capsule, length=7, radius=2.596, shapeName=dyn_coif_back_1_capsule, center=<1.8512 0 0>}]
 dyn_coif_back_2 [
  {shapeType=Capsule, length=0, radius=0, shapeName=dyn_coif_back_2_capsule}
  {shapeType=Box, extentX=7.5, extentY=2.0309, extentZ=10.6403, shapeName=dyn_coif_back_2_box, center=<0.6277 0.2406 0>}
 ]
 dyn_coif_back_base [{shapeType=Capsule, length=0.505, radius=0.505, shapeName=dyn_coif_back_base_capsule}]
 dyn_coif_side_l_1 [{shapeType=Capsule, length=7, radius=2, shapeName=dyn_coif_side_l_1_capsule, center=<1.9123 0 0>}]
 dyn_coif_side_l_2 [
  {shapeType=Capsule, length=0, radius=0, shapeName=dyn_coif_side_l_2_capsule}
  {
   shapeType=Box
   extentX=7.5
   extentY=2.7716
   extentZ=11.9338
   shapeName=dyn_coif_side_l_2_box
   center=<2.1488 0 0>
   rotation=<rot 0 0.0001 -6.4893>
  }
 ]
 dyn_coif_side_l_base [{shapeType=Capsule, length=0.505, radius=0.505, shapeName=dyn_coif_side_l_base_capsule}]
 dyn_coif_side_r_1 [{shapeType=Capsule, length=7, radius=2, shapeName=dyn_coif_side_r_1_capsule, center=<-2.3764 0 0>}]
 dyn_coif_side_r_2 [
  {shapeType=Capsule, length=0, radius=0, shapeName=dyn_coif_side_r_2_capsule}
  {
   shapeType=Box
   extentX=7.5
   extentY=2.792
   extentZ=11.9338
   shapeName=dyn_coif_side_r_2_box
   center=<-2.4592 0 0>
   rotation=<rot 0 0 -7.9777>
  }
 ]
 dyn_coif_side_r_base [{shapeType=Capsule, length=0.505, radius=0.505, shapeName=dyn_coif_side_r_base_capsule}]
 dyn_helmet_back_1 [{
   shapeType=Capsule
   length=3.4789
   radius=6.422
   shapeName=dyn_helmet_back_1_capsule
   center=<6.0988 0 -2.6186>
   rotation=<rot -46.6946 180 -180>
  }]
 dyn_helmet_back_base [{shapeType=Capsule, length=0.505, radius=0.505, shapeName=dyn_helmet_back_base_capsule, rotation=<rot 0 -90 0>}]
 dyn_helmet_frnt_1 [{shapeType=Capsule, length=4.7991, radius=5.2009, shapeName=dyn_helmet_frnt_1_capsule, center=<0 0 6.2437>}]
 dyn_helmet_frnt_base [{shapeType=Capsule, length=0.505, radius=0.505, shapeName=dyn_helmet_frnt_base_capsule, rotation=<rot 0 -90 0>}]
 dyn_helmet_mid_1_1 [{shapeType=Capsule, length=4.2219, radius=5.2885, shapeName=dyn_helmet_mid_1_1_capsule, center=<5.7771 0 0>, rotation=<rot -90 0 0>}]
 dyn_helmet_mid_1_base [{shapeType=Capsule, length=0.505, radius=0.505, shapeName=dyn_helmet_mid_1_base_capsule, rotation=<rot 0 -90 0>}]
 dyn_helmet_mid_2_1 [{shapeType=Capsule, length=4.9443, radius=5.0557, shapeName=dyn_helmet_mid_2_1_capsule, center=<5.4624 0 0>, rotation=<rot -90 0 0>}]
 dyn_helmet_mid_2_base [{shapeType=Capsule, length=0.505, radius=0.505, shapeName=dyn_helmet_mid_2_base_capsule, rotation=<rot 0 -90 0>}]
 neck_01 [{shapeType=Capsule, length=16.3911, radius=8.7567, shapeName=neck_01_capsule, center=<0.1298 0 0>, rotation=<rot 90 0 0>}]
 spine_05 [{shapeType=Capsule, length=13.0946, radius=17.8231, shapeName=spine_05_capsule}]
}
constraints [
 {bone1Name=dyn_coif_back_1, bone2Name=dyn_coif_back_base, swing1LimitDegrees=30, swing2LimitDegrees=5, twistLimitDegrees=0}
 {bone1Name=dyn_coif_back_2, bone2Name=dyn_coif_back_1, swing1LimitDegrees=20, swing2LimitDegrees=5, twistLimitDegrees=0}
 {bone1Name=dyn_coif_side_l_1, bone2Name=dyn_coif_side_l_base, swing1LimitDegrees=45, swing2LimitDegrees=5, twistLimitDegrees=0}
 {bone1Name=dyn_coif_side_l_2, bone2Name=dyn_coif_side_l_1, swing1LimitDegrees=20, swing2LimitDegrees=5, twistLimitDegrees=0}
 {bone1Name=dyn_coif_side_r_1, bone2Name=dyn_coif_side_r_base, swing1LimitDegrees=30, swing2LimitDegrees=5, twistLimitDegrees=0}
 {bone1Name=dyn_coif_side_r_2, bone2Name=dyn_coif_side_r_1, swing1LimitDegrees=20, swing2LimitDegrees=5, twistLimitDegrees=0}
 {bone1Name=dyn_helmet_back_1, bone2Name=dyn_helmet_back_base, swing1LimitDegrees=0, swing2LimitDegrees=5, twistLimitDegrees=0}
 {bone1Name=dyn_helmet_frnt_1, bone2Name=dyn_helmet_frnt_base, swing1LimitDegrees=20, swing2LimitDegrees=5, twistLimitDegrees=0}
 {bone1Name=dyn_helmet_mid_1_1, bone2Name=dyn_helmet_mid_1_base, swing1LimitDegrees=20, swing2LimitDegrees=5, twistLimitDegrees=0}
 {bone1Name=dyn_helmet_mid_2_1, bone2Name=dyn_helmet_mid_2_base, swing1LimitDegrees=20, swing2LimitDegrees=5, twistLimitDegrees=0}
 {
  bone1Name=dyn_helmet_back_base
  bone2Name=neck_01
  swing1LimitDegrees=45
  swing1Motion=Free
  swing2LimitDegrees=45
  swing2Motion=Free
  twistLimitDegrees=45
  twistMotion=Free
 }
 {
  bone1Name=dyn_helmet_frnt_base
  bone2Name=neck_01
  swing1LimitDegrees=45
  swing1Motion=Free
  swing2LimitDegrees=45
  swing2Motion=Free
  twistLimitDegrees=45
  twistMotion=Free
 }
 {
  bone1Name=dyn_helmet_mid_1_base
  bone2Name=neck_01
  swing1LimitDegrees=45
  swing1Motion=Free
  swing2LimitDegrees=45
  swing2Motion=Free
  twistLimitDegrees=45
  twistMotion=Free
 }
 {
  bone1Name=dyn_helmet_mid_2_base
  bone2Name=neck_01
  swing1LimitDegrees=45
  swing1Motion=Free
  swing2LimitDegrees=45
  swing2Motion=Free
  twistLimitDegrees=45
  twistMotion=Free
 }
 {
  bone1Name=dyn_coif_back_base
  bone2Name=neck_01
  swing1LimitDegrees=45
  swing1Motion=Free
  swing2LimitDegrees=45
  swing2Motion=Free
  twistLimitDegrees=45
  twistMotion=Free
 }
 {
  bone1Name=dyn_coif_side_l_base
  bone2Name=neck_01
  swing1LimitDegrees=45
  swing1Motion=Free
  swing2LimitDegrees=45
  swing2Motion=Free
  twistLimitDegrees=45
  twistMotion=Free
 }
 {
  bone1Name=dyn_coif_side_r_base
  bone2Name=neck_01
  swing1LimitDegrees=45
  swing1Motion=Free
  swing2LimitDegrees=45
  swing2Motion=Free
  twistLimitDegrees=45
  twistMotion=Free
 }
 {
  bone1Name=neck_01
  bone2Name=spine_05
  swing1LimitDegrees=45
  swing1Motion=Free
  swing2LimitDegrees=45
  swing2Motion=Free
  twistLimitDegrees=45
  twistMotion=Free
 }
]
