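# Animation Blueprint for the DenseFog large male face accessory, animating its helmet and coif.
# It copies the parent mesh pose, then runs a Control Rig fed crouch, skydive-float and rifle flags, with rifle covering shotguns too.
# Head-space RigidBody physics follows, using a gravity override from a deprecated state function.
# In the front end, the Frontend_Idle_Bypass curve blends DefaultSlot montages over the dynamic bones.
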AnimBlueprint /BRCosmetics/Characters/Player/Male/Large/Bodies/M_LRG_DenseFog/Meshes/Parts/M_LRG_DenseFog_FaceAcc_AnimBP
target_skeleton=&/M_LRG_DenseFog_FaceAcc_Skeleton
graphs [
 AnimGraph {nodes [
   n1 = Control Rig!(
    Source=Copy Pose From Mesh() {bUseAttachedParent=true}
    Alpha
    Is Crouching:double=Get Is Crouching() -> (Is Crouching:double)
    Is Skydive Floating:double=Get Is Skydive Floating() -> (Is Skydive Floating:double)
    Is Rifle:double=Get Is Rifle() -> (Is Rifle:double)
   ) -> (Pose) {
    alphaBoolBlend {blendOption=Linear}
    controlRigAssetReference {blueprintRigClass=&/M_LRG_DenseFog_FaceAcc_CtrlRig.M_LRG_DenseFog_FaceAcc_CtrlRig_C}
   }
   n2 = RigidBody(ComponentPose=Local To Component(LocalPose=@n1), ExternalForce=Get Gravity Override()) {
    bClampLinearTranslationLimitToRefPose=true
    bOverrideWorldGravity=true
    overridePhysicsAsset=&/M_LRG_DenseFog_FaceAcc_Physics
    simulationSpace=BaseBoneSpace
    alphaBoolBlend {blendOption=Linear}
    baseBoneRef {boneName=head}
    componentLinearAccScale=<0.3 0.3 0.3>
    componentLinearVelScale=<0.3 0.3 0.3>
    overrideWorldGravity {z=-980}
    simSpaceSettings {
     maxAngularAcceleration=500
     maxAngularVelocity=6
     maxLinearAcceleration=400
     maxLinearVelocity=250
     velocityScaleZ=0
     worldAlpha=1
     externalLinearDragV=<0.1 0.1>
    }
   }
   n3 = Layered blend per bone(
    BasePose=Component To Local(ComponentPose=@n2)
    BlendPoses_0=Slot 'DefaultSlot'()
    BlendWeights_0=Get Custome Idle is Playing Float() -> (Custome Idle is Playing Float:float)
   ) {layerSetup [{branchFilters [
       {boneName=dyn_helmet_back_base}
       {boneName=dyn_helmet_back_1}
       {boneName=dyn_helmet_mid_1_base}
       {boneName=dyn_helmet_mid_1_1}
       {boneName=dyn_coif_back_base}
       {boneName=dyn_coif_back_1}
       {boneName=dyn_coif_back_2}
       {boneName=dyn_coif_side_l_base}
       {boneName=dyn_coif_side_l_1}
       {boneName=dyn_coif_side_l_2}
       {boneName=dyn_coif_side_r_base}
       {boneName=dyn_coif_side_r_1}
       {boneName=dyn_coif_side_r_2}
      ]}]}
   Output Pose(Result=@n3) {layerGroup=None}
  ]}
 EventGraph {nodes [
   Event BlueprintUpdateAnimation()
   > n1 = Sequence!(execute) -> (then_0, then_1, then_2:Exec)
   > n2 = Gravity Override State Function DEPRECATED(
    CharacterPartAnimInstance=Self-Reference()
    GameTimeInSeconds=GetGameTimeInSeconds()
    GravityOverrideParams=Get Gravity Override Params 0() -> (Gravity Override Params 0:Gravity Override Params Struct Structure)
   )
   > Set Gravity Override(Gravity Override=@n2)
   n3 = Get bIsCrouching()
   n4 = Get bIsCrouchSprinting()
   > Set Is Crouching(Is Crouching:double="To Float (Boolean)"(InBool=OR Boolean(A=@n4, B=@n3))) -> (Output_Get:double)
   > Set Is Skydive Floating(Is Skydive Floating:double="To Float (Boolean)"(InBool=Get bIsSkydiveFloating())) -> (Output_Get:double)
   Branch(execute=@n1.then_2, Condition=Get bIsInFrontEnd())
   > Set Custome Idle is Playing Float(Custome Idle is Playing Float:float=GetCurveValue(self=GetOwnerFrontEndAnimInstance(), CurveName=Frontend_Idle_Bypass)) -> (Output_Get:float)
   n5 = Cast To FortPlayerAnimInstance(execute=@n1.then_1, Object=GetAnimInstance(self=GetOwnerSkeletalMeshComponent()))
   > n6 = Switch on EFortWeaponCoreAnimation!(execute, Selection:EFortWeaponCoreAnimation Enum=Get WeaponCoreAnim(self=@n5)) -> (
    Melee:Exec
    Pistol:Exec
    Shotgun:Exec
    PaperBlueprint:Exec
    Rifle:Exec
    MeleeOneHand:Exec
    MachinePistol:Exec
    RocketLauncher:Exec
    GrenadeLauncher:Exec
    GoingCommando:Exec
    AssaultRifle:Exec
    TacticalShotgun:Exec
    SniperRifle:Exec
    TrapPlacement:Exec
    ShoulderLauncher:Exec
    AbilityDecoTool:Exec
    Crossbow:Exec
    C4:Exec
    RemoteControl:Exec
    DualWield:Exec
    AR_BullPup:Exec
    AR_ForwardGrip:Exec
    MedPackPaddles:Exec
    SMG_P90:Exec
    AR_DrumGun:Exec
    Consumable_Small:Exec
    Consumable_Large:Exec
    Balloon:Exec
    MountedTurret:Exec
    CreativeTool:Exec
    ExplosiveBow:Exec
    AshtonIndigo:Exec
    AshtonChicago:Exec
    MeleeDualWield:Exec
    Unarmed:Exec
   )
   Set Is Rifle(
    execute=[
     @n6.AR_BullPup
     @n6.AR_DrumGun
     @n6.AR_ForwardGrip
     @n6.AbilityDecoTool
     @n6.AshtonChicago
     @n6.AshtonIndigo
     @n6.Balloon
     @n6.C4
     @n6.Consumable_Large
     @n6.Consumable_Small
     @n6.CreativeTool
     @n6.Crossbow
     @n6.DualWield
     @n6.ExplosiveBow
     @n6.GoingCommando
     @n6.GrenadeLauncher
     @n6.MachinePistol
     @n6.MedPackPaddles
     @n6.Melee
     @n6.MeleeDualWield
     @n6.MountedTurret
     @n6.PaperBlueprint
     @n6.Pistol
     @n6.RemoteControl
     @n6.RocketLauncher
     @n6.SMG_P90
     @n6.ShoulderLauncher
     @n6.SniperRifle
     @n6.TrapPlacement
     @n6.Unarmed
    ]
    Is Rifle:double=0
   ) -> (Output_Get:double)
   Set Is Rifle(execute=[@n6.AssaultRifle, @n6.Rifle, @n6.Shotgun, @n6.TacticalShotgun], Is Rifle=1)
  ]}
]
variables [
 Custome Idle is Playing Float {type=number}
 Gravity Override {type=Vector}
 Gravity Override Params 0 {type=GravityOverrideParamsStruct}
 Is Crouching {type=number}
 Is Rifle {type=number}
 Is Skydive Floating {type=number}
]
